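# Animation Blueprint for the Ice Flake backpack cosmetic that makes its six ice shard bones float procedurally.
# Each update builds three XY offset vectors from sine and cosine waves of game time at differing rates and amplitudes.
# The AnimGraph offsets the pack bone from ref pose, then adds those vectors in bone space, each shared by one left and one right shard.
# The flake_float helper and its Sequence are disconnected leftovers.
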AnimBlueprint /BRCosmetics/Accessories/FORT_Backpacks/Backpack_F_MED_Ice_Flake/Meshes/F_MED_Ice_Flake_AnimBlueprint
target_skeleton=&/F_MED_Ice_Flake_Skeleton
graphs [
 AnimGraph {nodes [
   n1 = "Transform (Modify) Bone - Bone: pack"(ComponentPose=Local To Component(LocalPose=Local Space Ref Pose()), Translation="0.000,-1.8,0.000", Rotation="0,0,0") {
    translationMode=BMM_Additive
    translationSpace=BCS_BoneSpace
    alphaBoolBlend {blendOption=Linear}
    boneToModify {boneName=pack}
    translation {y=-1.8}
   }
   n2 = "Transform (Modify) Bone - Bone: dyn_flake_r_01"(ComponentPose=@n1, Alpha=1, Translation=Get ShardFloatA() -> (ShardFloatA:Vector), Rotation="0,0,0", Scale="1,1,1") {translationMode=BMM_Additive, translationSpace=BCS_BoneSpace, alphaBoolBlend {blendOption=Linear}, boneToModify {boneName=dyn_flake_r_01}}
   n3 = "Transform (Modify) Bone - Bone: dyn_flake_r_02"(ComponentPose=@n2, Alpha=1, Translation=Get ShardFloatB() -> (ShardFloatB:Vector), Rotation="0,0,0", Scale="1,1,1") {translationMode=BMM_Additive, translationSpace=BCS_BoneSpace, alphaBoolBlend {blendOption=Linear}, boneToModify {boneName=dyn_flake_r_02}}
   n4 = "Transform (Modify) Bone - Bone: dyn_flake_r_03"(ComponentPose=@n3, Alpha=1, Translation=Get ShardFloatC() -> (ShardFloatC:Vector), Rotation="0,0,0", Scale="1,1,1") {translationMode=BMM_Additive, translationSpace=BCS_BoneSpace, alphaBoolBlend {blendOption=Linear}, boneToModify {boneName=dyn_flake_r_03}}
   n5 = "Transform (Modify) Bone - Bone: dyn_flake_l_02"(ComponentPose=@n4, Alpha=1, Translation=Get ShardFloatA(), Rotation="0,0,0", Scale="1,1,1") {translationMode=BMM_Additive, translationSpace=BCS_BoneSpace, alphaBoolBlend {blendOption=Linear}, boneToModify {boneName=dyn_flake_l_02}}
   n6 = "Transform (Modify) Bone - Bone: dyn_flake_l_01"(ComponentPose=@n5, Alpha=1, Translation=Get ShardFloatB(), Rotation="0,0,0", Scale="1,1,1") {translationMode=BMM_Additive, translationSpace=BCS_BoneSpace, alphaBoolBlend {blendOption=Linear}, boneToModify {boneName=dyn_flake_l_01}}
   n7 = "Transform (Modify) Bone - Bone: dyn_flake_l_03"(ComponentPose=@n6, Alpha=1, Translation=Get ShardFloatC(), Rotation="0,0,0", Scale="1,1,1") {translationMode=BMM_Additive, translationSpace=BCS_BoneSpace, alphaBoolBlend {blendOption=Linear}, boneToModify {boneName=dyn_flake_l_03}}
   Output Pose(Result=Component To Local(ComponentPose=@n7)) {layerGroup=None}
  ]}
 EventGraph {nodes [
   Event BlueprintUpdateAnimation()
   n1 = GetGameTimeInSeconds()
   n2 = "Sin (Radians)"(A=float * float(A=float * float(A=1, B=0.8), B=@n1))
   n3 = "Sin (Radians)"(A=float * float(A=float * float(A=1, B=1.2), B=@n1))
   > Set ShardFloatA(ShardFloatA:Vector=MakeVector(X=float * float(A=@n3, B=0.7), Y=float * float(A=@n2, B=0.5))) -> (Output_Get:Vector)
   n4 = "Cos (Radians)"!(A:double=float * float(A=float * float(A=1, B=0.8), B=@n1)) -> (ReturnValue:double)
   n5 = "Cos (Radians)"(A=float * float(A=float * float(A=1, B=0.7), B=@n1))
   > Set ShardFloatB(ShardFloatB:Vector=MakeVector(X=float * float(A=@n4, B=0.5), Y=float * float(A=@n5, B=0.25))) -> (Output_Get:Vector)
   n6 = "Sin (Radians)"(A=float * float(A=float * float(A=1, B=1.5), B=@n1))
   n7 = "Sin (Radians)"(A=float * float(A=float * float(A=1, B=1), B=@n1))
   > Set ShardFloatC(ShardFloatC:Vector=MakeVector(X=float * float(A=@n6, B=0.3), Y=float * float(A=@n7, B=0.6))) -> (Output_Get:Vector)
   n8 = Sequence!(execute) -> (then_0, then_1, then_2:Exec, then_3:Exec)
   > n9 = flake_float!(execute, self, Value:Vector="1,1, 0", Speed:Vector="1,1, 0") -> (then, Translation:Vector)
   n10 = flake_float(execute=@n8.then_1, Value="1,1, 0", Speed="1,1, 0")
   n11 = flake_float(execute=@n8.then_2, Value="1,1, 0", Speed="1,1, 0")
   TryGetPawnOwner()
   Set ShardFloatC(ShardFloatC=@n11)
   Set ShardFloatB(ShardFloatB=@n10)
   Set ShardFloatA(ShardFloatA=@n9)
  ]}
 flake_float {nodes [
   flake_float!() -> (then, Value:Vector=~, Speed:Vector=~) {is=title K2Node_FunctionEntry}
   > Return Node!(execute, Translation:Vector="0, 0, 0")
   n1 = GetGameTimeInSeconds()
   n2 = BreakVector(InVec="0, 0, 0")
   n3 = float * float(A=@n1, B=@n2.Z)
   n4 = BreakVector(InVec="0, 0, 0")
   n5 = float * float(A=@n1, B=@n2)
   n6 = float * float(A=@n1, B=@n2.Y)
   MakeVector(
    X=float * float(A="Sin (Degrees)"!(A:double=@n5) -> (ReturnValue:double), B=@n4)
    Y=float * float(A="Sin (Degrees)"(A=@n6), B=@n4.Y)
    Z=float * float(A="Sin (Degrees)"(A=@n3), B=@n4.Z)
   )
  ]}
]
variables [
 ShardFloatA {type=Vector}
 ShardFloatB {type=Vector}
 ShardFloatC {type=Vector}
]
